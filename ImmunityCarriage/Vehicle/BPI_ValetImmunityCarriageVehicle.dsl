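; Blueprint interface for the Valet Immunity Carriage vehicle, exposing access to parts of that vehicle to other Blueprints.
; It declares one function, GetSCMachineHoloLogoNiagaraComponent, whose name indicates it hands back the SC machine's holographic logo Niagara component.
; The file holds only the empty signature, so implementing Blueprints supply the actual component.
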(fn GetSCMachineHoloLogoNiagaraComponent
  (return 0))
