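; Drivable Reboot Van vehicle that resurrects players and revives downed pawns in a trunk seat.
; Spawn machine component states drive screen, hologram, rift, audio and gameplay cues; respawn FX wait for the new pawn.
; Seat events toggle interaction blocking, trace range and revive camera, with a looping timer completing revives.
; Server applies storm damage based on safe zone and occupancy, and updates map position on movement; boosting toggles siren and feedback.
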(event Custom|OnRep_RepDisabledState
  ;[1 1 1] Enabled/Disabled State
  (Variables|/Script/ImmunityCarriageRuntime.ImmunityCarriage_InteractionComponent|SetbAllowInteract (not (Variables|Default|GetRepDisabledState)) (Variables|Default|GetSpawnMachineComponent))
  ; Update hologram + screen enable/disable state
  (Components|Activation|SetActive (Variables|Default|GetNS_SCMachine_HoloLogo_LW) (not (Variables|Default|GetRepDisabledState)) false)
  (Utilities|IsValid (Variables|/Script/Engine.SkinnedMeshComponent|GetSkeletalMesh (Variables|Vehicle|GetSkeletalMesh))
    (:Is%20Valid
      (Rendering|Material|SetMaterial (Variables|Vehicle|GetSkeletalMesh) (Variables|Screen|GetScreenMaterialIndex) (select (Variables|Default|GetRepDisabledState) (Variables|Screen|GetStaticScreenMaterial) (Variables|Screen|GetScreenMaterial))))))

(event EventBeginPlay
  ;[0.04656 0.2083 0.0424] Begin play events
  (|CallParentFunction)
  (CallFunction|PlayRebootVanRiftedinEffect)
  (bind _multiinteractactorcomponent (Variables|Default|GetMultiInteractActorComponent)
    (:then
      (EventDispatchers|BindEventtoActorCountChangedDelegate _multiinteractactorcomponent (EventDispatchers|CreateEvent :!event Custom|HandleActorCountChanged)))
    (:else
      _))
  (Class|FortAsyncActionGameStateReady|GameStateReadyAsync
    (:OnReady
      (Vehicle|BindEventtoOnCurrentGearChangedDelegate (EventDispatchers|CreateEvent :!event Custom|HandleGearChange))
      (CallFunction|UpdateCancelSign)
      (CallFunction|UpdateStormDamage)
      ;[1 1 1] Enabled/Disabled State
      (Class|FortAsyncActionCurrentPlaylistReady|PlaylistDataReadyAsync
        (:OnReady
          (if (ScalableFloat|GetValueasBoolwithContext (Variables|Rows|GetRow_EnabledInWorld))
            (Resurrection|SetSpawnMachineState (Variables|Default|GetSpawnMachineComponent) "WaitingForUse" :!class /Script/ImmunityCarriageRuntime.ImmunityCarriage_InteractionComponent)
            ;[1 1 1] Bind to Player Spawned for Enhanced Resurrection
            (bind _asfort_game_mode_athena (Utilities|Casting|CastToFortGameModeAthena (Game|GetGameMode)))
            (EventDispatchers|BindEventtoOnPlayerSpawnDelegate _asfort_game_mode_athena (EventDispatchers|CreateEvent :!event HandlePlayerSpawned))
            (else
              (Variables|Default|SetRepDisabledState true))))))))

(event EventEndPlay
  (params (EndPlayReason :type EEndPlayReason%20Enum :default "Destroyed"))
  ;[0.2344 0.07324 0.0212] End Play: Make sure any spawned components are deactivated
  (|CallParentFunction EndPlayReason)
  (CallFunction|DeactivateFXComponents)
  (bind _rebootedplayerstate (Variables|Default|GetRebootedPlayerState)
    (:then
      (Events|UnbindEventfromOnPawnSet _rebootedplayerstate (EventDispatchers|CreateEvent :!event Custom|OnRespawnedPawnSet)))
    (:else)))

(event Custom|HandleGearChange
  (params (NewGear :type Integer))
  ;[1 1 1] Event ----> Handle any gear change fx/force feedback
  (bind _upshift (CallFunction|HandleShifting NewGear))
  ;; HandleFOVDuringShift
  (if _upshift
    (Variables|Camera|SetCameraFOVOffset -5.0)
    (Utilities|FlowControl|Delay 0.2)
    (Variables|Camera|SetCameraFOVOffset 0.0)))

(event EventPickRandomMaterial)

(event EventOnBoostStarted
  (bind _driverpawn (Variables|ValetDefaults|GetDriverPawn))
  (bind _returnvalue (Fortnite|GetFortPlayerControllerfromActor _driverpawn))
  ;[1 1 1] Boost Event > VFX > FF > State Set
  (|CallParentFunction)
  (CallFunction|ToggleSirenVFX true)
  (Utilities|IsValid _returnvalue
    (:Is%20Valid
      (Variables|Default|SetbHasBoosted true)
      (Game|Feedback|ClientPlayForceFeedback _returnvalue (Variables|Default|GetFF_CatchUpBoost) "None" false false false)
      (Projectile|AddGameplayCue%28Looping%29Local _returnvalue (Variables|Default|GetGC_CatchUpBoost) (Ability|GameplayCue|MakeGameplayCueParameters 0.0)))))

(event EventPawnEnterSeat
  (params (Vehicle :type Fort%20Vehicle%20Interface%20Interface%20%28by%20ref%29) (PlayerPawn :type Fort%20Pawn%20Object%20Reference) (SeatIndex :type Integer))
  (bind _returnvalue (Ability|GetAbilitySystemComponent PlayerPawn))
  (bind _returnvalue_2 (Fortnite|GetFortPlayerControllerfromActor PlayerPawn))
  (bind _self self)
  ; Enter Seta
  (Variables|Default|SetVehicleRef Vehicle)
  (Utilities|FlowControl|Sequence
    (:then_0
      (if (== SeatIndex (Variables|Default|GetReviveSeatIndex))
        (CallFunction|StartReviveUI PlayerPawn)
        ;;[1 1 1] Revive In Trunk
        (bind _returnvalue_3 (Utilities|Time|SetTimerbyEvent (delegate Custom|ReviveAfterTime) 0.1 true))
        (Variables|Default|SetReviveTimerHandle _returnvalue_3)
        (bind _output_get (Variables|Default|SetRevivePawn PlayerPawn))
        (bind _returnvalue_4 (GameplayEffects|ApplyGameplayEffectToSelf :self (Ability|GetAbilitySystemComponent _output_get) :GameplayEffectClass "/ImmunityCarriage/Gameplay/Effects/GE_ImmunityCarriage_BlockChangeSeatWhileRevive.GE_ImmunityCarriage_BlockChangeSeatWhileRevive_C" :Level 0.0))
        (Variables|Default|SetGE_BlockSeatChange _returnvalue_4)
        (CallFunction|PlayFX)
        (Ability|GameplayEffect|AddLooseGameplayTags (Variables|Default|GetRevivePawn) (Variables|Default|GetPawnRevivingTags))
        (Ability|GameplayEffect|AddLooseGameplayTags self (GameplayTags|MakeGameplayTagContainerfromTag (Variables|Default|GetVanRevivingTag)) true)
        (bind _fortplayercamerabase (CallFunction|GetCameraManager PlayerPawn))
        (Utilities|IsValid _fortplayercamerabase
          (:Is%20Valid
            (Camera|AddCameraModeOverride _fortplayercamerabase (Variables|Default|GetRezCamClass) (Variables|Default|GetRezTag) 100 _fortplayercamerabase 0)
            (Variables|Default|SetCachedReviveCamera _fortplayercamerabase)))))
    (:then_1
      ;;[1 1 1] Block Reboot If Shotgun
      (if (>= SeatIndex 2)
        (GameplayEffects|RemoveActiveEffectswithTags _returnvalue "(GameplayTags=((TagName=\"Gameplay.Action.Player.ImmunityCarriage.BlockedInteract\")))")
        (else
          (GameplayEffects|ApplyGameplayEffectToSelf :self _returnvalue :GameplayEffectClass "/ImmunityCarriage/Gameplay/Effects/GE_ImmunityCarriage_BlockInteract.GE_ImmunityCarriage_BlockInteract_C" :Level 0.0))))
    (:then_2
      ;;[1 1 1] Disable second interact trace if seating on the roof
      (if (or (== SeatIndex 2) (== SeatIndex 3))
        (Resurrection|EnableInteractionTracefromPlayer false _returnvalue_2)
        (Resurrection|SetInteractionExtraRange 144.0 _returnvalue_2)
        (else
          (Resurrection|EnableInteractionTracefromPlayer true _returnvalue_2)
          (Resurrection|SetInteractionExtraRange 0.0 _returnvalue_2))))
    (:then_3
      ;; Update Storm and Revive Seat
      (CallFunction|UpdateStormDamage)
      (CallFunction|UpdateCancelSign))
    (:then_4
      ;;[1 1 1] Out of fuel Audio
      (if (== SeatIndex 0)
        (Networking|CanExecuteCosmeticEvents
          (:True
            (if (GameplayTags|HasAnyMatchingGameplayTags _self (Variables|Default|GetOutOfFuelTag))
              (if (Class|FortVehicleAudioDataInterface|IsLocalforAudio _self)
                (Audio|PlaySound2D "/VehicleAudio/Cues/Veh_Valet_OutOfFuel_Cue.Veh_Valet_OutOfFuel_Cue")))))))))

(event EventOnBoostFinished
  (bind _driverpawn (Variables|ValetDefaults|GetDriverPawn))
  (bind _returnvalue (Fortnite|GetFortPlayerControllerfromActor _driverpawn))
  ;[1 1 1] Boost End Event > Remove VFX > Reset State || Delay is to make sure FOV reset back to normal so shifting doesnt affect it
  (|CallParentFunction)
  (CallFunction|ToggleSirenVFX false)
  (Utilities|IsValid _returnvalue
    (:Is%20Valid
      (Projectile|RemoveGameplayCue%28Looping%29Local _returnvalue (Variables|Default|GetGC_CatchUpBoost) (Ability|GameplayCue|MakeGameplayCueParameters 0.0))
      (Utilities|FlowControl|Delay 1.0)
      (Variables|Default|SetbHasBoosted false))))

(event Custom|ReviveAfterTime
  ; Enter Seta
  ;;[1 1 1] Revive In Trunk
  (bind _revivepawn (Variables|Default|GetRevivePawn)
    (:then
      (bind _asplayer_pawn_athena (Utilities|Casting|CastToPlayerPawn_Athena _revivepawn))
      (bind _multiinteractactorcomponent (Variables|/Game/Athena/PlayerPawn_Athena_Generic_Parent.PlayerPawn_Athena_Generic_Parent_C|GetMultiInteractActorComponent _asplayer_pawn_athena)
        (:then
          (if (> (Class|FortActorComponentMultiInteract|GetServerInteractionDuration _multiinteractactorcomponent) (Class|FortPlayerPawnAthena|GetRevivefromDBNOTime _asplayer_pawn_athena :!class /Script/FortniteGame.FortPlayerPawnAthena))
            (GameplayEffects|RemoveActiveGameplayEffect (Ability|GetAbilitySystemComponent _asplayer_pawn_athena) (Variables|Default|GetGE_BlockSeatChange) -1)
            (CallFunction|ReviveSuccess (Variables|Default|GetVehicleRef))
            (Variables|Default|SetRevivePawn 0)
            (Variables|Default|SetVehicleRef 0)))
        (:else)))
    (:else)))

(event EventHandleOnExplode
  (params (LastDamageInstigator :type Controller%20Object%20Reference) (Vehicle :type Actor%20Object%20Reference))
  ;[1 1 1] Event ----> Handle Lens Flare on explosion and headlight change
  (|CallParentFunction LastDamageInstigator Vehicle)
  (CallFunction|DeactivateFXComponents))

(event OnLightEnabledChanged(ValetLightSpawnerComponent)
  (params (enabled :type Boolean))
  ;[1 1 1] Event ----> Handle Lens Flare on explosion and headlight change
  (if enabled
    (bind _spawned_headlights_niagara (Variables|Default|GetSpawned%20Headlights%20Niagara)
      (:then
        (Components|Activation|Activate _spawned_headlights_niagara false)
        (Rendering|SetVisibility _spawned_headlights_niagara true false))
      (:else
        (bind _returnvalue (Niagara|SpawnSystemAttached (Variables|FX|GetHeadlights%20Niagara%20System) (Variables|Vehicle|GetSkeletalMesh) "root"))
        (Variables|Default|SetSpawned%20Headlights%20Niagara _returnvalue)
        (CallFunction|SetNumberOfFlarestoHeadlights _returnvalue)))
    (else
      (bind _spawned_headlights_niagara_2 (Variables|Default|GetSpawned%20Headlights%20Niagara)
        (:then
          (Components|Activation|Deactivate _spawned_headlights_niagara_2)
          (Rendering|SetVisibility _spawned_headlights_niagara_2 false false))
        (:else)))))

(event Vehicle|Cosmetics|EventOnVehicleCosmeticsFinished
  (bind _spawned_headlights_niagara (Variables|Default|GetSpawned%20Headlights%20Niagara))
  ;[1 1 1] Wait to load VCID Info and check if contains tag FUEL.ELECTRIC to set to native exhaust component
  (|CallParentFunction)
  (CallFunction|IsElectricCar)
  (Utilities|IsValid _spawned_headlights_niagara
    (:Is%20Valid
      (CallFunction|SetNumberOfFlarestoHeadlights _spawned_headlights_niagara))))

(event EventOnPawnEnterVehicle
  (params (PlayerPawn :type Fort%20Player%20Pawn%20Object%20Reference) (SeatIdx :type Integer))
  ;[1 1 1] Wait to load VCID Info and check if contains tag FUEL.ELECTRIC to set to native exhaust component
  (|CallParentFunction PlayerPawn SeatIdx)
  (Utilities|FlowControl|Sequence
    (:then_0
      (CallFunction|IsElectricCar)
      ;[1 1 1] BLOCK UI
      (if (not (> SeatIdx 1))
        (GameplayEffects|ApplyGameplayEffectToSelf :self (Ability|GetAbilitySystemComponent PlayerPawn) :GameplayEffectClass "/ImmunityCarriage/Gameplay/Effects/GE_ImmunityCarriage_BlockInteract.GE_ImmunityCarriage_BlockInteract_C" :Level 0.0))
      ;[1 1 1] Is Vehicle Occupied
      (if (> (Utilities|Array|Length (Vehicle|GetAllPassengers)) 0)
        ;[1 1 1] HOLOGRAM OFF
        (bind _ns_scmachine_holologo_lw (Variables|Default|GetNS_SCMachine_HoloLogo_LW)
          (:then
            (Components|Activation|Deactivate _ns_scmachine_holologo_lw)
            (Rendering|SetVisibility _ns_scmachine_holologo_lw false false))
          (:else))))
    (:then_1
      (Networking|SwitchHasAuthority
        (:Authority
          (GameplayEffects|ApplyGameplayEffectToSelf :self (Ability|GetAbilitySystemComponent PlayerPawn) :GameplayEffectClass (Variables|Quest|GetGE_ApplyInVehicleQuestTag) :Level 0.0))
        (:Remote
          (Rendering|Material|SetMaterial (Variables|Vehicle|GetSkeletalMesh) 0 "/ImmunityCarriage/Gameplay/Materials/RebootVan_Parts_Occlusion.RebootVan_Parts_Occlusion"))))))

(event Custom|UpdateCancelSign
  (bind _rezseatenabled (CallFunction|GetIsRezSeatEnabled))
  (bind _ns_immunitycarriage_cancel_sign (Variables|Default|GetNS_ImmunityCarriage_Cancel_Sign))
  (bind _ns_immunitycarriage_cancel_sign_2 (Variables|Default|GetNS_ImmunityCarriage_Cancel_Sign))
  ; Cancel Sign on Revive Seat
  (Networking|CanExecuteCosmeticEvents
    (:True
      (if (== (Variables|Default|GetbHologramEnabled) _rezseatenabled)
        _
        (elif _rezseatenabled
          (Niagara|SetNiagaraVariable(Bool) _ns_immunitycarriage_cancel_sign "EnableHologram" false)
          (Niagara|ResetSystem _ns_immunitycarriage_cancel_sign)
          (Variables|Default|SetbHologramEnabled true)
          (else
            (Utilities|IsValid (Vehicle|GetPawnatSeat (Variables|Default|GetReviveSeatIndex))
              (:Is%20Not%20Valid
                (Niagara|SetNiagaraVariable(Bool) _ns_immunitycarriage_cancel_sign_2 "EnableHologram" true)
                (Variables|Default|SetbHologramEnabled false)
                (if (Components|Activation|IsActive _ns_immunitycarriage_cancel_sign_2)
                  (Niagara|ResetSystem _ns_immunitycarriage_cancel_sign_2)
                  (else
                    (Rendering|SetVisibility _ns_immunitycarriage_cancel_sign_2 true false)
                    (Components|Activation|Activate _ns_immunitycarriage_cancel_sign_2 true) ; activate the component for the first time
                  ))))))))))

(event EventPawnExitSeat
  (params (Vehicle :type Fort%20Vehicle%20Interface%20Interface%20%28by%20ref%29) (PlayerPawn :type Fort%20Pawn%20Object%20Reference) (SeatIndex :type Integer))
  (bind _cachedrevivecamera (Variables|Default|GetCachedReviveCamera))
  ;[1 1 1] Exiting Seat
  (Utilities|FlowControl|Sequence
    (:then_0
      (if (== SeatIndex (Variables|Default|GetReviveSeatIndex))
        (Utilities|FlowControl|Sequence
          (:then_0
            ;; Must use cached camera pointer in case player died while reviving and can't get controller from pawn
            (Utilities|IsValid _cachedrevivecamera
              (:Is%20Valid
                (Camera|RemoveCameraModeOverride _cachedrevivecamera (Variables|Default|GetRezTag) _cachedrevivecamera false)
                (Variables|Default|SetCachedReviveCamera 0))))
          (:then_1
            (CallFunction|KillFx)
            (Ability|GameplayEffect|RemoveLooseGameplayTags PlayerPawn (Variables|Default|GetPawnRevivingTags))
            (Ability|GameplayEffect|RemoveLooseGameplayTags self (GameplayTags|MakeGameplayTagContainerfromTag (Variables|Default|GetVanRevivingTag)) true)
            (Utilities|Time|ClearandInvalidateTimerbyHandle (Variables|Default|GetReviveTimerHandle))
            (CallFunction|RemoveReviveUI PlayerPawn)))))
    (:then_1
      (CallFunction|UpdateStormDamage)
      (CallFunction|UpdateCancelSign))))

(event Custom|ToggleSirenVFX
  (params (ToggleOn :type Boolean))
  (if ToggleOn
    (Audio|Components|Audio|Play (Variables|Default|GetSiren) 0.0)
    (Rendering|Material|SetCustomPrimitiveDataVector2 (Variables|Vehicle|GetSkeletalMesh) 3 (Math|Vector2D|MakeVector2D 1.0 (Utilities|Time|GetGameTimeinSeconds)))
    (else
      (Audio|Components|Audio|Stop (Variables|Default|GetSiren))
      (Rendering|Material|SetCustomPrimitiveDataVector2 (Variables|Vehicle|GetSkeletalMesh) 3 (Math|Vector2D|MakeVector2D 0.0 (Utilities|Time|GetGameTimeinSeconds))))))

(event EventExplodeCosmetics
  (params (DestroyedActor :type Actor%20Object%20Reference))
  (|CallParentFunction DestroyedActor)
  (CallFunction|DeactivateFXComponents))

(event Custom|ReviveDamage
  (Utilities|FlowControl|DelayUntilNextTick)
  (bind _revivepawn (Variables|Default|GetRevivePawn)
    (:then
      (GameplayEffects|ApplyGameplayEffectToSelf :self (Ability|GetAbilitySystemComponent _revivepawn) :GameplayEffectClass (Variables|Default|GetGE_HealthSet) :Level 1.0))
    (:else)))

(event EventOnPawnExitVehicle
  (params (PlayerPawn :type Fort%20Player%20Pawn%20Object%20Reference) (ExitSocketName :type Name :default "None"))
  (bind _returnvalue (Ability|GetAbilitySystemComponent PlayerPawn))
  (bind _returnvalue_2 (Fortnite|GetFortPlayerControllerfromActor PlayerPawn))
  (|CallParentFunction PlayerPawn ExitSocketName)
  (Utilities|FlowControl|Sequence
    (:then_0
      ;[1 1 1] Allow interactions again
      (Utilities|IsValid PlayerPawn
        (:Is%20Valid
          (Utilities|IsValid _returnvalue
            (:Is%20Valid
              (GameplayEffects|RemoveActiveEffectswithTags _returnvalue "(GameplayTags=((TagName=\"Gameplay.Action.Player.ImmunityCarriage.BlockedInteract\")))"))))))
    (:then_1
      ;[1 1 1] Allow second interact trace again
      (Resurrection|EnableInteractionTracefromPlayer true _returnvalue_2)
      (Resurrection|SetInteractionExtraRange 0.0 _returnvalue_2))
    (:then_2
      ;[1 1 1] Is Vehicle Occupied
      (if (not (> (Utilities|Array|Length (Vehicle|GetAllPassengers)) 0))
        ;[1 1 1] HOLOGRAM ON
        (bind _ns_scmachine_holologo_lw (Variables|Default|GetNS_SCMachine_HoloLogo_LW)
          (:then
            (Components|Activation|Activate _ns_scmachine_holologo_lw false)
            (Rendering|SetVisibility _ns_scmachine_holologo_lw true false))
          (:else))))
    (:then_3
      (Networking|SwitchHasAuthority
        (:Authority
          (GameplayEffects|RemoveActiveEffectswithGrantedTags (Ability|GetAbilitySystemComponent PlayerPawn) (Variables|Quest|GetInVehicleQuestTag)))
        (:Remote
          (Rendering|Material|SetMaterial (Variables|Vehicle|GetSkeletalMesh) 0 0))))))

(event Custom|PlayFX
  (bind _selfrevivesound (Variables|Default|GetSelfReviveSound))
  (bind _skeletalmesh (Variables|Vehicle|GetSkeletalMesh))
  ;[1 0 0.9873 0.5] VFX
  (Networking|SwitchHasAuthority
    (:Remote
      (bind _returnvalue (Niagara|SpawnSystemAttachedwithParams (Utilities|Struct|MakeFXSystemSpawnParameters self "/ImmunityCarriage/FX/NS_ImmunityCarriage_ReviveScanner.NS_ImmunityCarriage_ReviveScanner" (Transformation|GetSocketLocation _skeletalmesh "ScanVFX") (Transformation|GetSocketRotation _skeletalmesh "ScanVFX") "1.000000,1.000000,1.000000" _skeletalmesh "ScanVFX" "KeepWorldPosition" false)))
      (Variables|Default|SetLoopFX _returnvalue)
      (bind _returnvalue_2 (Audio|SpawnSoundAttached _selfrevivesound _skeletalmesh)) ; REVIVE SFX
      (Variables|Default|SetLoopSound _returnvalue_2)
      (SoundIndicator|StartSoundIndicatorLooping (Variables|Default|GetSelfRevive%20Sound%20Indicator%20Component) (Transformation|GetActorLocation) _selfrevivesound 1.0 0)
      (Rendering|Material|SetCustomPrimitiveDataVector3 (Variables|Vehicle|GetSkeletalMesh) 3 (Math|Vector|MakeVector 1.0 (Utilities|Time|GetGameTimeinSeconds) 1.0)))))

(event Custom|KillFx
  (bind _loopfx (Variables|Default|GetLoopFX))
  ;[1 0 0.9873 0.5] VFX
  (Networking|SwitchHasAuthority
    (:Remote
      (Audio|Components|Audio|FadeOut (Variables|Default|GetLoopSound) 3.0 0.0 "Linear" :!class /Script/Engine.AudioComponent)
      (Rendering|SetVisibility _loopfx false false) ; safety to make sure it turns off
      (Components|Activation|Deactivate _loopfx)
      (Rendering|Material|SetCustomPrimitiveDataVector3 (Variables|Vehicle|GetSkeletalMesh) 3 (Math|Vector|MakeVector 0.0 (Utilities|Time|GetGameTimeinSeconds)))
      (SoundIndicator|StopSoundIndicatorLooping (Variables|Default|GetSelfRevive%20Sound%20Indicator%20Component)))))

(event Custom|SuccessFX
  ;[1 0 0.9873 0.5] VFX
  (Networking|SwitchHasAuthority
    (:Remote
      (Audio|SpawnSoundAttached "/ImmunityCarriage/Assets/Audio/SC_S39_ImmunityCarriage_MedCloud.SC_S39_ImmunityCarriage_MedCloud" (Variables|Vehicle|GetSkeletalMesh)) ; Med Cloud
    )))

(event Custom|UpdateStormDamage
  (bind _returnvalue (Ability|GetAbilitySystemComponent 0))
  ;[0.05232 0.15 0.03779 0.5] STORM DAMAGE UPDATE
  (Networking|SwitchHasAuthority
    (:Authority
      (if (Math|Boolean|ORBoolean (not (Variables|Default|GetbShouldTakeStormDamage)) (Variables|/Script/ImmunityCarriageRuntime.ImmunityCarriage_InteractionComponent|GetbIsInsideSafeZone (Variables|Default|GetSpawnMachineComponent)) (and (ScalableFloat|GetValueasBool (Variables|Default|GetHF_PreventStormDamageIfOccupied)) (CallFunction|IsVehicleOccupied)) (Utilities|IsValid (Vehicle|GetPawnatSeat (Variables|Default|GetReviveSeatIndex)) :!class /Script/Engine.KismetSystemLibrary))
        (GameplayEffects|RemoveActiveGameplayEffect _returnvalue (Variables|Default|GetGE_OutsideSafeZoneDamage) -1)
        (else
          (bind _returnvalue_2 (GameplayEffects|ApplyGameplayEffectToSelf :self _returnvalue :GameplayEffectClass "/Game/Athena/SafeZone/GE_OutsideSafeZoneDamage.GE_OutsideSafeZoneDamage_C" :Level 0.0))
          (bind _output_get (Variables|Default|SetGE_OutsideSafeZoneDamage _returnvalue_2))
          ;; Update Damage
          (GameplayEffects|UpdateActiveGameplayEffectSetbyCallerMagnitude (Ability|GetAbilitySystemComponent 0) _output_get "(TagName=\"SetByCaller.StormCampingDamage\")" (ScalableFloat|GetValueatLevel (Variables|Default|GetHF_StormDamage))))))))

(event OnPlayerPawnResurrected(SpawnMachineComponent)
  (params (PlayerPawn :type Fort%20Player%20Pawn%20Object%20Reference) (bEnhancedResurrection :type Boolean))
  (Utilities|FlowControl|Sequence
    (:then_0
      (Utilities|IsValid PlayerPawn
        (:Is%20Valid
          (CallFunction|MultiTriggerRespawnFX (Variables|/Script/Engine.Pawn|GetPlayerState PlayerPawn) bEnhancedResurrection)
          (Default|GiveWeaponsandResources (Fortnite|GetFortPlayerControllerfromActor PlayerPawn) bEnhancedResurrection)
          (bind _output_get (Variables|Default|SetRebootedPlayer PlayerPawn))
          ; Apply Quest Tag to the rebooted player
          (Fortnite|ApplyGameplayEffectsToActor _output_get (Utilities|Array|MakeArray (Variables|Quest|GetGE_RebootedPlayerQuestTag)) (0.0 0.0 0.0) false 0 (Ability|GetAbilitySystemComponentfromActor (Variables|/Script/Engine.Controller|GetPlayerState (Pawn|GetController _output_get))) "()")
          ;[1 1 1] After this call, the game mode applies the initial shield and health value to the respawned player. EnhancedResurrection may wanna set its own value and thus we're tracking the respawning players here so that we can set our values when the game mode tells us that the player spawn process is complete.
          ; See HandlePlayerSpawned().
          (if bEnhancedResurrection
            (Utilities|Array|Add (Variables|Default|GetEnhancedResurrectionRebootingPlayerPawns) PlayerPawn)
            (Utilities|FlowControl|Delay 0.0)
            (Ability|SendGameplayEventtoActor PlayerPawn "(TagName=\"Event.Player.Reboot.EnhancedResurrection\")")))))
    (:then_1
      (for _array_element (Variables|Default|GetPlayers%20Interacting)
        (Ability|SendGameplayEventtoActor (Pawn|GetPlayerPawnorVehicleDriver _array_element) "(TagName=\"Event.Player.Reboot.DidReboot\")")))))

(event Custom|Multi_TriggerRespawnFX
  (params (RessedPlayerState :type Player%20State%20Object%20Reference) (bEnhancedResurrection :type Boolean))
  (bind _returnvalue (PlayerState|GetPawn RessedPlayerState))
  ;[0.02658 0.15 0.1075 0.5] Trigger respawn FX on all machines when new pawn becomes available
  (Variables|Default|SetEnhancedResFX bEnhancedResurrection)
  (bind _output_get (Variables|Default|SetRebootedPlayerState RessedPlayerState))
  (Utilities|IsValid _output_get
    (:Is%20Valid
      (Utilities|IsValid _returnvalue
        (:Is%20Valid
          (CallFunction|RespawnFX _returnvalue (Variables|Default|GetEnhancedResFX)))
        (:Is%20Not%20Valid
          (Events|BindEventtoOnPawnSet RessedPlayerState (delegate Custom|OnRespawnedPawnSet)))))))

(event Custom|Respawn%20FX
  (params (PlayerPawn :type Pawn%20Object%20Reference) (EnhancedRessurection :type Boolean))
  (bind _returnvalue (select EnhancedRessurection "(TagName=\"SCMachine.Respawn.EnhancedResurrection\")" 0))
  (bind _returnvalue_2 (GameplayTags|MakeGameplayTagContainerfromTag _returnvalue))
  (bind (_returnvalue_x _returnvalue_y _returnvalue_z) (Math|Vector|BreakVector (Transformation|GetWorldLocation (Variables|Default|GetNS_SCMachine_HoloLogo_LW))))
  ;[0.1449 0.07141 0.15 0.5] Respawn FX
  (bind _asplayer_pawn_athena (Utilities|Casting|CastToPlayerPawn_Athena PlayerPawn))
  (Utilities|FlowControl|Sequence
    (:then_0
      (GameplayCue|ExecuteGameplayCueOnActor(Burst) self (Variables|GameplayCues|GetActivateGameplayCue) (Ability|GameplayCue|MakeGameplayCueParameters :NormalizedMagnitude 0.0 :RawMagnitude 0.0 :AggregatedSourceTags _returnvalue_2 :Location (Math|Vector|MakeVectorNetQuantize10 _returnvalue_x _returnvalue_y _returnvalue_z) :Normal "0,0,0" :SourceObject _asplayer_pawn_athena :GameplayEffectLevel 1 :AbilityLevel 1 :bReplicateLocationWhenUsingMinimalRepProxy false))
      (GameplayCue|ExecuteGameplayCue%28Burst%29Local _asplayer_pawn_athena (Variables|GameplayCues|GetPlayerRespawnGameplayCue) (Ability|GameplayCue|MakeGameplayCueParameters :NormalizedMagnitude 0.0 :RawMagnitude 0.0 :AggregatedTargetTags _returnvalue_2 :Location "0,0,0" :Normal "0,0,0" :GameplayEffectLevel 1 :AbilityLevel 1 :bReplicateLocationWhenUsingMinimalRepProxy false))
      (Class|PlayerPawnAthena|ActivateInvincibleVisuals _asplayer_pawn_athena (Variables|Default|GetInvincibleVisualsDuration))
      (CallFunction|UpdateScreen (Resurrection|GetSpawnMachineState (Variables|Default|GetSpawnMachineComponent) :!class /Script/ImmunityCarriageRuntime.ImmunityCarriage_InteractionComponent))
      (CallFunction|SpawnRift _asplayer_pawn_athena))
    (:then_1
      (Default|CallOnPawnResurrected _asplayer_pawn_athena EnhancedRessurection))))

(event Custom|OnRespawnedPawnSet
  (params (Player :type Player%20State%20Object%20Reference) (NewPawn :type Pawn%20Object%20Reference) (OldPawn :type Pawn%20Object%20Reference))
  ;[0.02658 0.15 0.1075 0.5] Trigger respawn FX on all machines when new pawn becomes available
  (Utilities|IsValid Player
    (:Is%20Valid
      (Utilities|IsValid NewPawn
        (:Is%20Valid
          (Events|UnbindEventfromOnPawnSet Player (delegate Custom|OnRespawnedPawnSet))
          (CallFunction|RespawnFX NewPawn (Variables|Default|GetEnhancedResFX)))))))

(event OnResurrectionStarted(SpawnMachineComponent)
  (params (PlayerController :type Fort%20Player%20Controller%20Athena%20Object%20Reference))
  (Variables|Default|SetRepServerInteract false))

(event OnResurrectionCompleted(SpawnMachineComponent))

(event OnBeginCooldown(SpawnMachineComponent)
  (Utilities|Array|Clear (Variables|Default|GetPlayers%20Interacting)))

(event OnEndCooldown(SpawnMachineComponent))

(event OnSpawnMachineStateChanged(SpawnMachineComponent)
  (params (NewState :type ESpawnMachineState%20Enum :default "Default"))
  ;[0.04287 0.15 0.1088 0.5] On State Change
  (Development|PrintString (Utilities|String|Append "On Spawn Machine State Changed: " (Utilities|String|EnumtoString NewState)))
  (CallFunction|HandleAmbientSound NewState)
  (CallFunction|UpdateScreen NewState)
  (CallFunction|UpdateActivatingRiftState NewState)
  (CallFunction|SetHologramActive (or (Utilities|Enum|Equal(Enum) NewState "OnCooldown") (Utilities|Enum|Equal(Enum) NewState "WaitingForUse")))
  (if (Utilities|Enum|Equal(Enum) NewState "Active")
    (GameplayCue|AddGameplayCue self (Variables|GameplayCues|GetActiveGameplayCue))
    (Variables|Default|SetDid%20Horn%20Play? false)
    (CallFunction|PlayHornSound)
    (Utilities|FlowControl|RetriggerableDelay (Variables|Default|GetHonkResetDelay))
    (Variables|Default|SetDid%20Horn%20Play? false)
    (else
      (if (not (Utilities|Time|IsTimerActivebyHandle (Variables|Default|GetRemoveActiveGameplayCueTimerHandle)))
        (bind _returnvalue (Utilities|Time|SetTimerbyEvent (delegate Custom|RemoveActiveGameplayCue) (Variables|Default|GetRemoveActiveGameplayCueDelay)))
        (Variables|Default|SetRemoveActiveGameplayCueTimerHandle _returnvalue))
      (CallFunction|UpdateCooldownSound))))

(event Custom|RemoveActiveGameplayCue
  ;[0.04287 0.15 0.1088 0.5] On State Change
  (GameplayCue|RemoveGameplayCue self (Variables|GameplayCues|GetActiveGameplayCue)))

(event OnBeginLongUse(SpawnMachineComponent)
  (params (AthenaPC :type Fort%20Player%20Controller%20Athena%20Object%20Reference))
  ;[0.05004 0.15 0.1269 0.5] Server Interact
  (Utilities|Array|AddUnique (Variables|Default|GetPlayers%20Interacting) AthenaPC)
  (Development|PrintString "Event - On Begin Long use")
  (Variables|Default|SetRepServerInteract true))

(event OnInterruptLongUse(SpawnMachineComponent)
  (params (AthenaPC :type Fort%20Player%20Controller%20Athena%20Object%20Reference))
  ;[0.05004 0.15 0.1269 0.5] Server Interact
  (Utilities|Array|RemoveItem (Variables|Default|GetPlayers%20Interacting) AthenaPC)
  (Development|PrintString "Event - On Interrupt Long Use")
  (if (Utilities|IsArrayNonEmpty (Variables|Default|GetPlayers%20Interacting))
    ;;[1 0.6602 0.2944] When MultiInteract, only notify stop interaction when no players interacting
    (if (Class|FortActorComponentMultiInteract|IsMultipleInteractionEnabled (Variables|Default|GetMultiInteractActorComponent))
      _
      (else
        _))
    (else
      _))
  (Variables|Default|SetRepServerInteract false))

(event EventOnDeathPlayEffects
  (params (Damage :type Float%20%28single-precision%29) (DamageTags :type Gameplay%20Tag%20Container%20Structure%20%28by%20ref%29) (Momentum :type Vector :default "0, 0, 0") (HitInfo :type Hit%20Result%20Structure%20%28by%20ref%29) (InstigatedBy :type Fort%20Pawn%20Object%20Reference) (DamageCauser :type Actor%20Object%20Reference) (EffectContext :type Gameplay%20Effect%20Context%20Handle%20Structure))
  ; Disable interaction when the vehicle dies
  (Variables|/Script/ImmunityCarriageRuntime.ImmunityCarriage_InteractionComponent|SetbAllowInteract false (Variables|Default|GetSpawnMachineComponent)))

(event Custom|HandleActorCountChanged
  (params (InteractingActorCount :type Integer) (MultiInteractMultiplier :type Float%20%28single-precision%29) (InteractingDuration :type Float%20%28single-precision%29))
  ; Sync Screens with MultiItneract Component
  (Networking|SwitchHasAuthority
    (:Remote
      (if (and (Class|FortActorComponentMultiInteract|IsMultipleInteractionEnabled (Variables|Default|GetMultiInteractActorComponent)) (> InteractingActorCount 0))
        (Rendering|Material|SetScalarParameterValue (Variables|Screen|GetScreenMaterial) "BaseTime" InteractingDuration :!class /Script/Engine.MaterialInstanceDynamic)
        (Rendering|Material|SetScalarParameterValue (Variables|Screen|GetScreenMaterial) "InteractTime" (Utilities|Time|GetGameTimeinSeconds) :!class /Script/Engine.MaterialInstanceDynamic)
        (Rendering|Material|SetScalarParameterValue (Variables|Screen|GetScreenMaterial) "Multiplier" MultiInteractMultiplier :!class /Script/Engine.MaterialInstanceDynamic)
        (CallFunction|SetActivatingRiftPlayrate MultiInteractMultiplier)))))

(event Custom|Handle%20Ambient%20Sound
  (params (SpawnMachineState :type ESpawnMachineState%20Enum :default "Default"))
  (bind _ambientsound (Variables|Default|GetAmbientSound))
  ;[0.09504 0.00015 0.15 0.5] Audio Handling
  ;;[0.15 0.04667 0.09434 0.5] Ambient Sound
  (switch Utilities|FlowControl|Switch|SwitchonESpawnMachineState SpawnMachineState
    (:WaitingForUse
      (Audio|Components|Audio|FadeIn _ambientsound 4.0 1.0 0.0 "Linear" :!class /Script/Engine.AudioComponent))
    (:BeingUsed
      (if (not (Audio|Components|Audio|IsPlaying _ambientsound))
        (Audio|Components|Audio|FadeIn (Variables|Default|GetAmbientSound) 4.0 1.0 0.0 "Linear" :!class /Script/Engine.AudioComponent)))
    (:Active
      _)
    (:Complete
      _)
    (:OnCooldown
      _))
  (Audio|Components|Audio|FadeOut (Variables|Default|GetAmbientSound) 1.0 0.0 "Linear" :!class /Script/Engine.AudioComponent))

(event Custom|Start%20Active%20Sound%20Timer
  ;[0.09504 0.00015 0.15 0.5] Audio Handling
  ;;[0.15 0.04454 0.07508 0.5] Active Sound Timer
  (Utilities|FlowControl|RetriggerableDelay (ScalableFloat|GetValueatLevelwithContext (Variables|Rows|GetRow_InteractTime)))
  (GameplayCue|RemoveGameplayCue self (Variables|GameplayCues|GetActivatingGameplayCue)))

(event Custom|Play%20Horn%20Sound
  (bind _sound_hornalert (Variables|FX|GetSound_HornAlert))
  ;[0.09504 0.00015 0.15 0.5] Audio Handling
  ;;[0.04828 0.04169 0.15 0.5] Horn
  (if (Utilities|Enum|Equal(Enum) (Resurrection|GetSpawnMachineState (Variables|Default|GetSpawnMachineComponent) :!class /Script/ImmunityCarriageRuntime.ImmunityCarriage_InteractionComponent) "Active")
    (if (not (Variables|Default|GetDid%20Horn%20Play?))
      (Audio|SpawnSoundAttached _sound_hornalert (Variables|Transformation|GetRootComponent))
      (bind _returnvalue (Development|GetConsoleVariableFloatValue "SpawnMachine.SoundIndicator.MaxDistanceMultiplier"))
      (SoundIndicator|ShowSoundIndicator (Variables|Default|GetReboot%20Sound%20Indicator%20Component) (Transformation|GetActorLocation) _sound_hornalert _returnvalue 0)
      (Variables|Default|SetDid%20Horn%20Play? true))))

(event Custom|Update%20Cooldown%20Sound
  ;[0.09504 0.00015 0.15 0.5] Audio Handling
  ;;[0.07186 0.09779 0.15 0.5] Cooldown
  (if (Utilities|Enum|Equal(Enum) (Resurrection|GetSpawnMachineState (Variables|Default|GetSpawnMachineComponent) :!class /Script/ImmunityCarriageRuntime.ImmunityCarriage_InteractionComponent) "OnCooldown")
    (GameplayCue|AddGameplayCue self (Variables|GameplayCues|GetCooldownGameplayCue))
    (else
      (GameplayCue|RemoveGameplayCue self (Variables|GameplayCues|GetCooldownGameplayCue)))))

(event Custom|Spawn%20Rift
  (params (InFortPawn :type Fort%20Pawn%20Object%20Reference))
  ;[0.15 0.05839 0.1352 0.5] Reboot VFX handling
  ;;[0.149 0.03111 0.15 0.5] Spawn Rift: Spawn rift particle system above van where player spawns in
  (Utilities|IsValid InFortPawn
    (:Is%20Valid
      (bind _mesh (Variables|/Script/Engine.Character|GetMesh InFortPawn)
        (:then
          (Niagara|SpawnSystematLocation (Variables|FX|GetNS%20Rift%20Spawn) (+ (Transformation|GetWorldLocation _mesh) (Variables|FX|GetRiftFXOffset))))
        (:else)))))

(event Custom|Set%20Hologram%20Active
  (params (TurnOn :type Boolean))
  (bind (_returnvalue_x _returnvalue_y _returnvalue_z) (Math|Vector|BreakVector (Transformation|GetWorldLocation (Variables|Default|GetNS_SCMachine_HoloLogo_LW))))
  (bind _ns_scmachine_holologo_lw (Variables|Default|GetNS_SCMachine_HoloLogo_LW))
  ;[0.15 0.05839 0.1352 0.5] Reboot VFX handling
  ;;[0.06615 0.15 0.1356 0.5] Toggle Hologram: Update hologram visiblity with tranistion FX
  (if (Math|Boolean|NotEqual(Boolean) TurnOn (Components|Activation|IsActive _ns_scmachine_holologo_lw))
    (Components|Activation|SetActive _ns_scmachine_holologo_lw TurnOn false)
    (Networking|SwitchHasAuthority
      (:Authority
        (GameplayCue|ExecuteGameplayCueOnActor(Burst) self (Variables|GameplayCues|GetHologramTransitionGameplayCue) (Ability|GameplayCue|MakeGameplayCueParameters :NormalizedMagnitude 0.0 :RawMagnitude 0.0 :EffectContext "()" :AggregatedSourceTags "(GameplayTags=)" :AggregatedTargetTags "(GameplayTags=)" :Location (Math|Vector|MakeVectorNetQuantize10 _returnvalue_x _returnvalue_y _returnvalue_z) :Normal "0,0,0" :GameplayEffectLevel 1 :AbilityLevel 1 :bReplicateLocationWhenUsingMinimalRepProxy false))
        (Networking|ForceNetUpdate)))))

(event Custom|Update%20Activating%20Rift%20State
  (params (SpawnMachineState :type ESpawnMachineState%20Enum :default "Default"))
  ;[0.15 0.05839 0.1352 0.5] Reboot VFX handling
  ;;[0.03111 0.02186 0.15 0.5] Update Activating Rift State: Update the rift that replaces the hologram while reboot interaction is in progress
  (if (Utilities|Enum|Equal(Enum) SpawnMachineState "BeingUsed")
    (Utilities|IsValid (Variables|Default|GetPortalClosingComponent)
      (:Is%20Not%20Valid
        (bind _returnvalue (Niagara|SpawnSystemAttachedwithParams (Utilities|Struct|MakeFXSystemSpawnParameters self (Variables|FX|GetReduxPortalClosingSystem) (Variables|FX|GetPortalClosingFXOffset) "0.000000,0.000000,0.000000" "1.000000,1.000000,1.000000" (Variables|Default|GetNS_SCMachine_HoloLogo_LW) "None" "SnapToTarget")))
        (Variables|Default|SetPortalClosingComponent _returnvalue)
        (CallFunction|SetActivatingRiftPlayrate 1.0)))
    (else
      (bind _portalclosingcomponent (Variables|Default|GetPortalClosingComponent)
        (:then
          (Components|DestroyComponent _portalclosingcomponent)
          (Variables|Default|SetPortalClosingComponent 0))
        (:else)))))

(event Custom|Set%20Activating%20Rift%20Playrate
  (params (Playrate :type Float%20%28double-precision%29 :default 1.0))
  (bind _row_interacttime (Variables|Rows|GetRow_InteractTime))
  (bind _returnvalue (ScalableFloat|GetValueatLevel _row_interacttime 1.0))
  (bind _returnvalue_2 (Math|Float|SafeDivide 1e+01 _returnvalue))
  (bind _returnvalue_3 (* Playrate _returnvalue_2))
  ;[0.15 0.05839 0.1352 0.5] Reboot VFX handling
  ;;[0.15 0.03886 0.09769 0.5] Set Activating Rift Playrate: Update how quickly activating rift plays its animation
  (bind _portalclosingcomponent (Variables|Default|GetPortalClosingComponent)
    (:then
      (Niagara|SetCustomTimeDilation _portalclosingcomponent _returnvalue_3)
      (Niagara|SetNiagaraVariable(Float) _portalclosingcomponent "Time Scale" _returnvalue_3))
    (:else)))

(event Custom|Deactivate%20FX%20Components
  (bind _ns_scmachine_holologo_lw (Variables|Default|GetNS_SCMachine_HoloLogo_LW))
  ;[0.15 0.05839 0.1352 0.5] Reboot VFX handling
  ;;[0.15 0.03078 0.02218 0.5] Cleanup FX
  (Components|Activation|Deactivate _ns_scmachine_holologo_lw)
  (Rendering|SetVisibility _ns_scmachine_holologo_lw false false)
  (bind _spawned_headlights_niagara (Variables|Default|GetSpawned%20Headlights%20Niagara)
    (:then
      (Components|DestroyComponent _spawned_headlights_niagara))
    (:else
      _))
  (bind _ambientsound (Variables|Default|GetAmbientSound)
    (:then
      (Audio|Components|Audio|FadeOut _ambientsound 0.5 0.0 "Linear" :!class /Script/Engine.AudioComponent))
    (:else
      _))
  (bind _portalclosingcomponent (Variables|Default|GetPortalClosingComponent)
    (:then
      (Components|DestroyComponent _portalclosingcomponent))
    (:else)))

(event EventBP_OnServerMove
  (params (MoveDist :type Float%20%28single-precision%29))
  ; This is necessary for replicating the position so remote clients can see the vans move in the in-game map
  (Class|ImmunityCarriageFunctionLibrary|HandleSpawnMachineMoved (Variables|Default|GetSpawnMachineComponent)))

(event OnSafeZoneStatusChangedEvent(SpawnMachineComponent)
  (params (bIsInSafeZone :type Boolean))
  (CallFunction|UpdateStormDamage)
  (CallFunction|UpdateCancelSign))

(event Custom|RiftRebootVanIntoWorld
  ;[0.5 0 0.25] Set up for rift effect on server when spawning. This is called from the reinjection gamestate component
  (Networking|SwitchHasAuthority
    (:Authority
      ;; This is only called on server so we need to replicate (initial only) if we were rifted into the world
      (Variables|Default|SetVanWasRiftedIn true)
      (Vehicle|WakeUp)
      (GameplayEffects|ApplyGameplayEffectToSelf :self (Ability|GetAbilitySystemComponent 0) :GameplayEffectClass "/ImmunityCarriage/Gameplay/Effects/GE_ImmunityCarriage_ReiftedIntoWorld.GE_ImmunityCarriage_ReiftedIntoWorld_C" :Level 0.0))))

(event Custom|PlayRebootVanRiftedInEffect
  (bind _returnvalue (Transformation|GetActorLocation))
  (bind _riftfxoffset (Variables|FX|GetRiftFXOffset))
  (bind _returnvalue_2 (+ _returnvalue _riftfxoffset))
  ;[0.5 0 0.25] Play a rift effect if the reboot van was rifted into the world
  (Networking|CanExecuteCosmeticEvents
    (:True
      (Variables|Default|GetVanWasRiftedIn
        (:then
          (Niagara|SpawnSystematLocation (Variables|FX|GetNS%20Rift%20Spawn) _returnvalue_2)
          (Audio|PlaySoundatLocation (Variables|Default|GetRiftInSound) _returnvalue_2)
          (Utilities|FlowControl|Delay 1.0)
          (Audio|SpawnSoundAttached :Sound "/ImmunityCarriage/Assets/Audio/SC_S39_ImmunityCarriage_Horn_Cue.SC_S39_ImmunityCarriage_Horn_Cue" :AttachToComponent (Variables|Vehicle|GetSkeletalMesh) :AttachPointName "None" :LocationType "KeepRelativeOffset" :bStopWhenAttachedToDestroyed false :VolumeMultiplier 2.0 :PitchMultiplier 1.0 :StartTime 0.0 :bAutoDestroy true) ; Horn
        )
        (:else)))))

(event ImmunityCarriage_PawnResurrectionInterface|EventBindToOnPawnResurrectedEvent
  (params (OnPawnResurrected :type Delegate%20%28by%20ref%29))
  (Default|BindEventtoOnPawnResurrected OnPawnResurrected))

(event ImmunityCarriage_PawnResurrectionInterface|EventUnbindFromOnPawnResurrectedEvent
  (params (OnPawnResurrected :type Delegate%20%28by%20ref%29))
  (Default|UnbindEventfromOnPawnResurrected OnPawnResurrected))

(fn IsSeatEnabled
  (params (SeatIdx :type Integer))
  (return false))

(fn BP_ShouldBlockInteractForSeat
  (params (SeatIndex :type Integer))
  (return (not (or (== SeatIndex 2) (== SeatIndex 3)))))

(fn CleanupDriftGE
  (GameplayEffects|RemoveActiveGameplayEffectbySourceEffect (Ability|GetAbilitySystemComponent (Variables|ValetDefaults|GetDriverPawn)) (Variables|Drift|GetDriftingTag) 0 -1)
  (for _array_element (Vehicle|GetAllPassengers)
    (Utilities|IsValid _array_element
      (:Is%20Valid
        (GameplayEffects|RemoveActiveGameplayEffectbySourceEffect (Ability|GetAbilitySystemComponent _array_element) (Variables|Drift|GetDriftingTag) 0 -1))))
  (return))

(fn DestroyExitBlockingActors
  (bind _skeletalmesh (Variables|Vehicle|GetSkeletalMesh))
  (if (ScalableFloat|GetValueasBool (Variables|Rows|GetRow_DestroyExitPath))
    (LocalVariables|SetExitSocketLocation (Transformation|GetSocketLocation _skeletalmesh "BackExit"))
    (LocalVariables|SetVehicleCenterLocation (Transformation|GetSocketLocation _skeletalmesh "Center"))
    (bind _outhits (Collision|MultiLineTraceForObjects (LocalVariables|GetVehicleCenterLocation) (LocalVariables|GetExitSocketLocation) (LocalVariables|GetObjectTypeQuery) false (Utilities|Array|MakeArray (Variables|Default|GetRevivePawn))))
    (for _array_element _outhits
      (bind _asbuilding_actor (Utilities|Casting|CastToBuildingActor (.HitActor (Collision|BreakHitResult _array_element))))
      (if (Variables|/Script/Engine.Actor|GetbCanBeDamaged _asbuilding_actor)
        (Gameplay|Die _asbuilding_actor "(TagName=\"Gameplay.Effect.InstantDeath.Environment\")" 0 0 :!class /Script/FortniteGame.BuildingActor)))
    (bind _outactors (Collision|CapsuleOverlapActors (LocalVariables|GetExitSocketLocation) 42.0 8e+01 (LocalVariables|GetObjectTypeQuery) "/Script/FortniteGame.BuildingSMActor"))
    (for _array_element_2 _outactors
      (bind _asbuilding_actor_2 (Utilities|Casting|CastToBuildingActor _array_element_2))
      (if (Variables|/Script/Engine.Actor|GetbCanBeDamaged _asbuilding_actor_2)
        (Gameplay|Die _asbuilding_actor_2 "(TagName=\"Gameplay.Effect.InstantDeath.Environment\")" 0 0 :!class /Script/FortniteGame.BuildingActor)))))

(fn GetCameraManager
  (params (PlayerPawn :type Actor%20Object%20Reference))
  (bind _returnvalue (Fortnite|GetFortPlayerControllerfromActor PlayerPawn))
  (Utilities|IsValid _returnvalue
    (:Is%20Valid
      (bind _asfort_player_camera_base (Utilities|Casting|CastToFortPlayerCameraBase (Variables|/Script/Engine.PlayerController|GetPlayerCameraManager _returnvalue))
        (:then
          (return _asfort_player_camera_base))
        (:CastFailed
          _)))
    (:Is%20Not%20Valid
      _))
  (return 0))

(fn GetIsRezSeatEnabled
  (return (and (Variables|/Script/ImmunityCarriageRuntime.ImmunityCarriage_InteractionComponent|GetbIsInsideSafeZone (Variables|Default|GetSpawnMachineComponent)) (> (ScalableFloat|GetCachedScalableFloatValue (Variables|Rows|GetRow_ReviveEnabled)) 0))))

(fn GetSCMachineHoloLogoNiagaraComponent
  (return (Variables|Default|GetNS_SCMachine_HoloLogo_LW)))

(fn GiveWeaponsAndResources
  (params (FortPC :type Fort%20Player%20Controller%20Object%20Reference) (bEnhancedResurrection :type Boolean))
  ;[0.15 0.06724 0.1493 0.5] GiveWeaponsAndResources: Player has respawned, give them the desired loot.
  (Utilities|IsValid FortPC
    (:Is%20Valid
      (Utilities|FlowControl|Sequence
        (:then_0
          (Variables|Default|SetTierGroupName (Variables|Default|GetGrantedItemsLootTier))
          (if bEnhancedResurrection
            (bind _returnvalue (Fortnite|Athena|GameState|GetGameStateAthena))
            (Utilities|IsValid _returnvalue
              (:Is%20Valid
                (bind _returnvalue_2 (Actor|GetComponentbyClass _returnvalue "/Script/FortniteGame.FortGameStateComponent_Reboot"))
                (Utilities|IsValid _returnvalue_2
                  (:Is%20Valid
                    (Variables|Default|SetTierGroupName (Resurrection|GetEnhancedResurrectionLootTierGroupName _returnvalue_2))))))))
        (:then_1
          (bind (_outloottodrop _returnvalue_3) (Loot|PickLootDrops (Variables|Default|GetTierGroupName) 0 -1 "(GameplayTags=)"))
          (if _returnvalue_3
            (for _array_element _outloottodrop
              (bind (_array_element_itemdefinition _array_element_itemguid _array_element_count) (Class|FortItemFunctionLibrary|BreakFortItemEntry _array_element))
              (bind _asfort_world_item_definition (Utilities|Casting|CastToFortWorldItemDefinition _array_element_itemdefinition))
              (Inventory|GiveItemtoInventoryOwner :InventoryOwner FortPC :ItemDefinition _asfort_world_item_definition :NumberToGive _array_element_count :bNotifyPlayer true :ItemLevel -1 :PickupInstigatorHandle -1 :bUseItemPickupAnalyticEvent false :WeaponAmmoOverride -1))))))))

(fn HandlePlayerSpawned
  (params (Pawn :type Fort%20Player%20Pawn%20Object%20Reference))
  ;[0.0273 0.15 0.1433 0.5] Apply enhanced resurrection gameplay effects to the respawned player.
  (Utilities|IsValid Pawn
    (:Is%20Valid
      (bind _returnvalue (Utilities|Array|RemoveItem (Variables|Default|GetEnhancedResurrectionRebootingPlayerPawns) Pawn))
      (if _returnvalue
        (Fortnite|ApplyGameplayEffectsToActor Pawn (Variables|Default|GetEnhancedResurrectionGameplayEffects) (0.0 0.0 0.0) false 0 (Ability|GetFortAbilitySystemComponent Pawn) "()")))))

(fn HandleShifting
  (params (NewGear :type Integer))
  (bind _driverpawn (Variables|ValetDefaults|GetDriverPawn))
  (bind _returnvalue (Fortnite|GetFortPlayerControllerfromActor _driverpawn))
  (bind _driverpawn_2 (Variables|ValetDefaults|GetDriverPawn))
  (bind _returnvalue_2 (Fortnite|GetFortPlayerControllerfromActor _driverpawn_2))
  (Utilities|IsValid (Variables|ValetDefaults|GetDriverPawn)
    (:Is%20Valid
      (if (> NewGear (Variables|ShiftingGears|GetCurrentGearAudio))
        (Development|PrintString "UpShift")
        (Game|Feedback|ClientPlayForceFeedback _returnvalue_2 (Variables|ShiftingGears|GetFF_Upshift) "None" false false false)
        (Game|Feedback|ClientStartCameraShake _returnvalue_2 (Variables|ShiftingGears|GetUpShiftCameraShake) 1.0 "CameraLocal" (Transformation|GetActorRotation _driverpawn_2))
        (Variables|ShiftingGears|SetCurrentGearAudio NewGear)
        (return true)
        (else
          (Development|PrintString "Downshift")
          (Game|Feedback|ClientPlayForceFeedback _returnvalue (Variables|ShiftingGears|GetFF_Downshift) "None" false false false)
          (Game|Feedback|ClientStartCameraShake _returnvalue (Variables|ShiftingGears|GetDownShiftCameraShake) 1.0 "CameraLocal" (Transformation|GetActorRotation))
          (Variables|ShiftingGears|SetCurrentGearAudio NewGear)
          (return false))))))

(fn InitializeHeadlightAndLensFlareFX
  (bind _result (Math|Integer|IncrementInt (Variables|Lights|GetUpdateValetHeadlightColors))
    (:%20%20
      (Variables|Lights|SetUpdateValetHeadlightColors _result)
      ;[1 1 1] Override value based on child needs
      (bind _spawned_headlights_niagara (Variables|Default|GetSpawned%20Headlights%20Niagara)
        (:then
          (Niagara|SetNiagaraVariable(Int32) _spawned_headlights_niagara "SedanColor" (Variables|Lights|GetLensFlare_SedanColor)))
        (:else)))))

(fn InitializeTailLightColor
  (Niagara|SetNiagaraVariable(Int32) (Variables|Valet|RoadAndExhaust|Niagara|GetNativeComp_RoadAndExhaust) "SedanColor" (Variables|Default|GetRoadAndExhaust_TailLight_SedanColor)))

(fn IsElectricCar
  (bind _returnvalue (GameplayTags|HasMatchingGameplayTag "(TagName=\"Vehicle.Fuel.Electric\")"))
  (Niagara|SetNiagaraVariable(Bool) (Variables|Valet|RoadAndExhaust|Niagara|GetNativeComp_RoadAndExhaust) "IsElectric" _returnvalue)
  (return _returnvalue))

(fn OnRep_RepServerInteract
  (bind _returnvalue (Utilities|Time|GetGameTimeinSeconds))
  (bind _row_interacttime (Variables|Rows|GetRow_InteractTime))
  (bind _returnvalue_2 (ScalableFloat|GetValueatLevelwithContext _row_interacttime))
  (bind _ns_scmachine_holologo_lw (Variables|Default|GetNS_SCMachine_HoloLogo_LW))
  ;[0.1246 0.03233 0.15 0.5] Interaction FX
  (if (Variables|Default|GetRepServerInteract)
    ;;[0.02342 0.15 0.04594 0.5] Interacting
    (Utilities|FlowControl|Sequence
      (:then_0
        ;;;[0.05856 0.07534 0.15 0.5] Screen + Hologram VFX
        (Rendering|Material|SetScalarParameterValue (Variables|Screen|GetScreenMaterial) (Variables|FXParameters|GetParam_Screen_InteractTime) _returnvalue :!class /Script/Engine.MaterialInstanceDynamic)
        (Rendering|Material|SetScalarParameterValue (Variables|Screen|GetScreenMaterial) (Variables|FXParameters|GetParam_Screen_InteractDuration) _returnvalue_2 :!class /Script/Engine.MaterialInstanceDynamic))
      (:then_1
        (Effects|Components|ParticleSystem|SetFloatParameter _ns_scmachine_holologo_lw (Variables|FXParameters|GetParam_Hologram_Gametime) _returnvalue)
        (Effects|Components|ParticleSystem|SetFloatParameter _ns_scmachine_holologo_lw (Variables|FXParameters|GetParam_Hologram_InteractDuration) _returnvalue_2))
      (:then_2
        ;;;[0.04148 0.15 0.0605 0.5] Activating FX
        (if (or (Utilities|Enum|Equal(Enum) (Resurrection|GetSpawnMachineState (Variables|Default|GetSpawnMachineComponent) :!class /Script/ImmunityCarriageRuntime.ImmunityCarriage_InteractionComponent) "WaitingForUse") (Utilities|Enum|Equal(Enum) (Resurrection|GetSpawnMachineState (Variables|Default|GetSpawnMachineComponent) :!class /Script/ImmunityCarriageRuntime.ImmunityCarriage_InteractionComponent) "BeingUsed"))
          (GameplayCue|AddGameplayCue self (Variables|GameplayCues|GetActivatingGameplayCue))
          (CallFunction|StartActiveSoundTimer)))
      (:then_3
        ;;; Initiate MultiInteract's screens mat params
        (if (Class|FortActorComponentMultiInteract|IsMultipleInteractionEnabled (Variables|Default|GetMultiInteractActorComponent))
          (LocalVariables|SetMultiInteractWeight 1.0)
          (else
            (LocalVariables|SetMultiInteractWeight 0.0)))
        (Rendering|Material|SetScalarParameterValue (Variables|Screen|GetScreenMaterial) "NewMultiWeight" (LocalVariables|GetMultiInteractWeight) :!class /Script/Engine.MaterialInstanceDynamic)))
    (else
      ;;[0.15 0.04949 0.05092 0.5] Not interacting
      (Rendering|Material|SetScalarParameterValue (Variables|Screen|GetScreenMaterial) (Variables|FXParameters|GetParam_Screen_InteractTime) 0.0 :!class /Script/Engine.MaterialInstanceDynamic)
      (Effects|Components|ParticleSystem|SetFloatParameter (Variables|Default|GetNS_SCMachine_HoloLogo_LW) (Variables|FXParameters|GetParam_Hologram_Gametime) 0.0)
      (GameplayCue|RemoveGameplayCue self (Variables|GameplayCues|GetActivatingGameplayCue)))))

(fn OnRep_UpdateValetHeadlightColors
  (if (not (Networking|IsServer))
    ;[1 1 1] Override value based on child needs
    (ValetLights|SetValetLightColor (Variables|Default|GetValetLightSpawnerComponent) (Variables|Lights|GetHeadLightColor))))

(fn OverrideIsSeatEnabled
  (params (SeatIndex :type Integer) (bIsSeatEnabled :type Boolean))
  (if (== SeatIndex (Variables|Default|GetReviveSeatIndex))
    (if (Variables|/Script/ImmunityCarriageRuntime.ImmunityCarriage_InteractionComponent|GetbIsInsideSafeZone (Variables|Default|GetSpawnMachineComponent))
      (return (and (> (ScalableFloat|GetCachedScalableFloatValue (Variables|Rows|GetRow_ReviveEnabled)) 0) bIsSeatEnabled))
      (else
        (return false)))
    (else
      (return bIsSeatEnabled))))

(fn RemoveReviveUI
  (params (PawnToBeRevived :type Fort%20Pawn%20Object%20Reference))
  (bind _asfort_player_pawn_athena (Utilities|Casting|CastToFortPlayerPawnAthena PawnToBeRevived))
  (Class|FortPlayerPawnAthena|RemoveDBNORevivingActor _asfort_player_pawn_athena _asfort_player_pawn_athena)
  (Utilities|IsValid (Actor|GetComponentbyClass _asfort_player_pawn_athena "/Game/Athena/Items/Components/MultiInteractActorComponent.MultiInteractActorComponent_C")
    (:Is%20Valid
      (Class|FortActorComponentMultiInteract|RemoveInteractingActor (Actor|GetComponentbyClass _asfort_player_pawn_athena "/Game/Athena/Items/Components/MultiInteractActorComponent.MultiInteractActorComponent_C") self))))

(fn ReviveSuccess
  (params (InputPin :type Fort%20Vehicle%20Interface%20Interface%20%28by%20ref%29))
  (bind _revivepawn (Variables|Default|GetRevivePawn))
  (bind _revivepawn_2 (Variables|Default|GetRevivePawn))
  (bind _asfort_athena_vehicle (Utilities|Casting|CastToFortAthenaVehicle InputPin))
  (LocalVariables|SetVehicle _asfort_athena_vehicle)
  (bind _asfort_player_pawn_athena (Utilities|Casting|CastToFortPlayerPawnAthena _revivepawn_2))
  (bind _output_get (LocalVariables|SetRevivedPlayerAthenaPawn _asfort_player_pawn_athena))
  (bind _asfort_player_pawn (Utilities|Casting|CastToFortPlayerPawn _revivepawn_2))
  (LocalVariables|SetRevivedPlayerPawn _asfort_player_pawn)
  (Class|FortPlayerPawnAthena|ForceRevivefromDBNO _output_get 0)
  ;[1 1 1] Need to Remove UI Tell
  (CallFunction|ReviveDamage)
  (CallFunction|SuccessFX)
  (Utilities|FlowControl|Sequence
    (:then_0
      (bind _returnvalue (Vehicle|GetVehicleSeatComponent self))
      (bind _output_get_2 (LocalVariables|SetVehicleSeats (Variables|/Script/FortniteGame.FortVehicleSeatComponent|GetPlayerSlots _returnvalue)))
      (for (_ _array_index) _output_get_2
        ;[1 1 1] NOT TRUNK
        (if (!= _array_index (Variables|Default|GetReviveSeatIndex))
          (Utilities|IsValid (Vehicle|GetPawnatSeat self _array_index)
            (:Is%20Not%20Valid
              (LocalVariables|SetFoundSeat _array_index))))))
    (:then_1
      ;[1 1 1] Empty Passenger Seat?
      (if (not (<= (LocalVariables|GetFoundSeat) 0))
        (Vehicle|RequestChangeSeatonServer (LocalVariables|GetVehicle) (Pawn|GetController (LocalVariables|GetRevivedPlayerPawn)) -1) ; Requesting seat -1 will give next available seat (driver if empty)
        ; Is the pawn still stuck in the back? If so, then eject the player.
        (if (== (Class|FortVehicleInterface|FindSeatIndex (LocalVariables|GetVehicle) (LocalVariables|GetRevivedPlayerPawn)) (Variables|Default|GetReviveSeatIndex))
          _))
      ; Ensure the path to the back is clear of obstacles like buildings
      (Default|DestroyExitBlockingActors)
      (Vehicle|EjectPlayerPawn (LocalVariables|GetVehicle) (LocalVariables|GetRevivedPlayerPawn)))
    (:then_2
      (Ability|GameplayEffect|RemoveLooseGameplayTags (Variables|Default|GetRevivePawn) (Variables|Default|GetPawnRevivingTags)))
    (:then_3
      ; Analytics
      (if (Networking|HasAuthority _revivepawn)
        (bind _asfort_player_controller_athena (Utilities|Casting|CastToFortPlayerControllerAthena (Fortnite|GetFortPlayerControllerfromActor _revivepawn)))
        (Resurrection|PlayerRevivedinRebootVan _asfort_player_controller_athena)))))

(fn SetNumberOfFlaresToHeadlights
  (params (Niagara%20system :type Niagara%20Particle%20System%20Component%20Object%20Reference))
  (bind _existingsocketslist (LocalVariables|GetExistingSocketsList))
  (bind _returnvalue (Utilities|Array|Length _existingsocketslist))
  (bind _asfort_vehicle_skel_mesh_component (Utilities|Casting|CastToFortVehicleSkelMeshComponent (Variables|Vehicle|GetSkeletalMesh)))
  (for _array_element (LocalVariables|GetFilteredSocketsList)
    (if (Transformation|DoesSocketExist _asfort_vehicle_skel_mesh_component _array_element)
      (Utilities|Array|Add (LocalVariables|GetExistingSocketsList) _array_element)
      (Niagara|SetNiagaraVariable(Int32) Niagara%20system "FlaresCount" _returnvalue)
      (Development|PrintString (Utilities|String|Append "Car lensflares Headligth: " (Utilities|String|ToString(Integer) _returnvalue)) true true "(R=1.000000,G=0.000000,B=0.145770,A=1.000000)" 4.0))))

(fn ShouldShowSoundIndicator
  (params (PlayerController :type Fort%20Player%20Controller%20Object%20Reference))
  (return (Math|Boolean|ORBoolean (Utilities|IsValid (Vehicle|GetPawnatSeat 0) :!class /Script/Engine.KismetSystemLibrary) (and (!= (Game|Player|GetPlayerPawn PlayerController) (Vehicle|GetPawnatSeat (Variables|Default|GetReviveSeatIndex))) (Utilities|IsValid (Vehicle|GetPawnatSeat (Variables|Default|GetReviveSeatIndex)) :!class /Script/Engine.KismetSystemLibrary)) (and (Utilities|Enum|Equal(Enum) (Resurrection|GetSpawnMachineState (Variables|Default|GetSpawnMachineComponent) :!class /Script/FortniteGame.SpawnMachineFunctionLibrary) "BeingUsed") (not (Class|FortActorComponentMultiInteract|HasInteractingActor (Variables|Default|GetMultiInteractActorComponent) (Game|Player|GetPlayerPawn PlayerController)))))))

(fn StartReviveUI
  (params (PawnToBeRevived :type Fort%20Pawn%20Object%20Reference))
  (bind _asfort_player_pawn_athena (Utilities|Casting|CastToFortPlayerPawnAthena PawnToBeRevived))
  (Class|FortPlayerPawnAthena|AddDBNORevivingActor _asfort_player_pawn_athena _asfort_player_pawn_athena)
  (Utilities|IsValid (Actor|GetComponentbyClass _asfort_player_pawn_athena "/Game/Athena/Items/Components/MultiInteractActorComponent.MultiInteractActorComponent_C")
    (:Is%20Valid
      (Class|FortActorComponentMultiInteract|AddInteractingActor (Actor|GetComponentbyClass _asfort_player_pawn_athena "/Game/Athena/Items/Components/MultiInteractActorComponent.MultiInteractActorComponent_C") self))))

(fn UpdateScreen
  (params (SpawnMachineState :type ESpawnMachineState%20Enum :default "Default"))
  (bind _returnvalue (Utilities|Time|GetGameTimeinSeconds))
  (bind _spawnmachinecomponent (Variables|Default|GetSpawnMachineComponent))
  (bind _returnvalue_2 (Resurrection|GetSpawnMachineCooldownLength _spawnmachinecomponent))
  (bind _ns_scmachine_holologo_lw (Variables|Default|GetNS_SCMachine_HoloLogo_LW))
  (bind _screenmaterial (Variables|Screen|GetScreenMaterial))
  (bind _spawnmachinestate (Variables|GetSpawnMachineState))
  (Rendering|Material|SetScalarParameterValue _screenmaterial (Variables|FXParameters|GetParam_Screen_Spawning) (Math|Conversions|ToFloat(Boolean) (Utilities|Enum|Equal(Enum) _spawnmachinestate "Active")) :!class /Script/Engine.MaterialInstanceDynamic) ; Spawning state
  (Rendering|Material|SetScalarParameterValue _screenmaterial (Variables|FXParameters|GetParam_Screen_Cooldown) (Math|Conversions|ToFloat(Boolean) (Utilities|Enum|Equal(Enum) _spawnmachinestate "OnCooldown")) :!class /Script/Engine.MaterialInstanceDynamic) ; Cooldown state
  (switch Utilities|FlowControl|Switch|SwitchonESpawnMachineState (Variables|GetSpawnMachineState)
    (:WaitingForUse
      (Fortnite|SetNiagaraAssetMatchActivation (Variables|Default|GetNS_SCMachine_HoloLogo_LW) (Variables|FX|GetHologramFX_WaitingForUse) false)
      (Effects|Components|ParticleSystem|SetFloatParameter (Variables|Default|GetNS_SCMachine_HoloLogo_LW) (Variables|FXParameters|GetParam_Hologram_Gametime) 0.0))
    (:Complete
      _)
    (:OnCooldown
      (Utilities|FlowControl|Sequence
        (:then_0
          _)
        (:then_1
          (Utilities|FlowControl|Sequence
            (:then_0
              (Rendering|Material|SetScalarParameterValue (Variables|Screen|GetScreenMaterial) (Variables|FXParameters|GetParam_Screen_CooldownTime) _returnvalue :!class /Script/Engine.MaterialInstanceDynamic)
              (Rendering|Material|SetScalarParameterValue (Variables|Screen|GetScreenMaterial) (Variables|FXParameters|GetParam_Screen_CooldownDuration) _returnvalue_2 :!class /Script/Engine.MaterialInstanceDynamic))
            (:then_1
              (Fortnite|SetNiagaraAssetMatchActivation _ns_scmachine_holologo_lw (Variables|FX|GetHologramFX_OnCooldown) false)
              (Effects|Components|ParticleSystem|SetFloatParameter _ns_scmachine_holologo_lw (Variables|FXParameters|GetParam_Hologram_Gametime) _returnvalue)
              (Effects|Components|ParticleSystem|SetFloatParameter _ns_scmachine_holologo_lw (Variables|FXParameters|GetParam_Hologram_InteractDuration) _returnvalue_2)))))))
  (Rendering|Material|SetScalarParameterValue (Variables|Screen|GetScreenMaterial) (Variables|FXParameters|GetParam_Screen_SpawningTime) 0.0 :!class /Script/Engine.MaterialInstanceDynamic))

(fn ConstructionScript
  (|CallParentFunction)
  (Variables|Default|SetbShouldTakeStormDamage true)
  (Vehicle|OverridePassengersCanInteract true)
  ; Initialize screen material
  (Utilities|IsValid (Variables|/Script/Engine.SkinnedMeshComponent|GetSkeletalMesh (Variables|Vehicle|GetSkeletalMesh))
    (:Is%20Valid
      (bind _output_get (Variables|Screen|SetScreenMaterialIndex (Rendering|Material|GetMaterialIndex (Variables|Vehicle|GetSkeletalMesh) "Screens")))
      (bind _returnvalue (Rendering|Material|CreateDynamicMaterialInstance (Variables|Vehicle|GetSkeletalMesh) _output_get (Rendering|Material|GetMaterialbyName (Variables|Vehicle|GetSkeletalMesh) "Screens") "None" :!class /Script/Engine.PrimitiveComponent))
      (Variables|Screen|SetScreenMaterial _returnvalue))))
